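# Physics asset for the Grilled Cheese Male two-handed melee pickaxe, giving its left and right stick joints physical bodies.
# A long capsule body covers the handle bone, and a small offset capsule sits on each stick bone.
# Each stick is constrained to the handle with a swing limit and locked twist, so it can pivot but not spin about its axis.
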PhysicsAsset /BRCosmetics/Weapons/FORT_Melee/Pickaxe_Grilled_Cheese_Male/Meshes/Grilled_Cheese_Male_Physics
bodies {
 Melee_TwoHanded_Handle [{shapeType=Capsule, length=44.75, radius=3.5}]
 dyn_L_stick_jnt [{shapeType=Capsule, length=1, radius=2, center=<2.6 0 0>}]
 dyn_R_stick_jnt [{shapeType=Capsule, length=1, radius=2, center=<2.6 0 0>}]
}
constraints [
 {
  bone1Name=dyn_L_stick_jnt
  bone2Name=Melee_TwoHanded_Handle
  swing1LimitDegrees=45
  swing2LimitDegrees=45
  twistLimitDegrees=45
  twistMotion=Locked
 }
 {
  bone1Name=dyn_R_stick_jnt
  bone2Name=Melee_TwoHanded_Handle
  swing1LimitDegrees=45
  swing2LimitDegrees=45
  twistLimitDegrees=45
  twistMotion=Locked
 }
]
