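# Niagara effect for the Grilled Cheese pickaxe that draws flickering, lightning-like ribbons and glows on its skeletal mesh.
# Two randomly delayed looping BoneFlow emitters, burst and rate-based, partition particles into ribbons along mesh sockets, bowed by a radius curve.
# Each ribbon group randomly takes User.TheColor or fixed yellow, with curve-driven brightness flashes and material parameters.
# Glow and SubUV lightning sprites sample the same mesh; the Decal emitter is disabled.
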NiagaraSystem /BRCosmetics/Weapons/FORT_Melee/Pickaxe_Grilled_Cheese_Male/VFX/NS_Pickaxe_GrilledCheese
dependencies {
 usedDataInterfaces [&NiagaraDataInterfaceCurve, &NiagaraDataInterfaceColorCurve, &NiagaraDataInterfaceVectorCurve]
 usedDynamicInputs [/Niagara/DynamicInputs/UniformRange/UniformRangedVector4]
 usedModules [&SystemState]
}
emitters {
 BoneFlow_Burst {
  emitterSpawnScript [
   /SetVariables(Emitter.RandomSeed:NiagaraInt32=/Niagara/DynamicInputs/Random/RandomInt(Randomness Mode:ENiagaraRandomnessMode=::NewEnumerator0, Random Seed:NiagaraInt32=0, Fixed Random Seed:NiagaraBool=false)) {isSetParameters=true, module=SetVariables_#1}
   /Game/Effects/Niagara/Modules/Partition/FN_PartitionAttributes(Particles Per Partition=6, Partitions=4)
  ]
  emitterUpdateScript [
   &EmitterState(
    Life Cycle Mode=::NewEnumerator1
    Loop Delay=&RandomRangeFloat(Minimum=0.15, Maximum=0.8)
    Loop Duration=0.3
    Recalculate Duration Each Loop=true
    UseLoopDelay=true
   )
   &SpawnBurst_Instantaneous(Spawn Count=24, Use Spawn Probability=true)
   /SetVariables(Emitter.RandomBoolGroups:Vector4f=/Niagara/DynamicInputs/UniformRange/V2/RandomRangeVector4(
     Minimum:Vector4f=<0 0 0 0>
     Maximum:Vector4f=<1 1 1 1>
     Evaluation Type:ENiagaraRandomnessEvaluation=::NewEnumerator0
     Recalculate Random Each Loop:NiagaraBool=true
     Randomness Mode:ENiagaraRandomnessMode=::NewEnumerator0
     Random Seed:NiagaraInt32=0
     Fixed Random Seed:NiagaraBool=false
    )) {isSetParameters=true, module=SetVariables_#2}
   /Game/Effects/Niagara/Modules/Skydiving/FN_Emitter_SkydiveAttributes() {module=Emitter_SkydiveAttributes}
  ]
  particleSpawnScript [
   &FN_SetSpawnGroup()
   /Game/Effects/Niagara/Modules/Partition/FN_Particle_partition_v5/FN_Particle_Partition_v5(Enable Ribbon ID=true, Pattern:ENiagaraPartitionWaveForm=::NewEnumerator1, Seed:NiagaraInt32=$Emitter.RandomSeed)
   /Game/Effects/Niagara/Modules/BoneFlow/FN_Initialize_BoneFlow_Attributes_v2(
    Angle:NiagaraFloat=&RandomRangeFloat(Minimum=0, Maximum=1)
    Radius:NiagaraFloat=&FloatFromCurve(
     FloatCurve=DataInterface() {propertyValues {Curve {keys [
         {arriveTangent=5.6652, interpMode=RCIM_Cubic, leaveTangent=5.6652, tangentMode=RCTM_User, time=0, value=0}
         <at 0.5 1 RCIM_Cubic>
         {arriveTangent=-5.8125, interpMode=RCIM_Cubic, leaveTangent=-5.8125, tangentMode=RCTM_User, time=1, value=0}
        ]}}}
     CurveIndex=$Particles.Partition.NormalizedID
     Scale Curve=&RandomRangeFloat(Minimum=5, Maximum=10)
    )
   )
   &InitializeRibbon(
    Color=&MaskLinearColorBySpawnGroup(
     Masked Linear Color=&MakeCustomLinearColorFromBool(
      Bool=&SetBoolByFloatComparison(A=&MakeFloatFromVector4(Vector 4=$Emitter.RandomBoolGroups, Channel=::NewEnumerator0), B=0.5)
      True Color=$User.TheColor
      False Color=<rgba 4 3.5 1 0>
     )
     Passthrough Linear Color=&MaskLinearColorBySpawnGroup(
      Masked Linear Color=&MakeCustomLinearColorFromBool(
       Bool=&SetBoolByFloatComparison(A=&MakeFloatFromVector4(Vector 4=$Emitter.RandomBoolGroups, Channel=::NewEnumerator1), B=0.5)
       True Color=$User.TheColor
       False Color=<rgba 4 3.5 1 0>
      )
      Passthrough Linear Color=&MaskLinearColorBySpawnGroup(
       Masked Linear Color=&MakeCustomLinearColorFromBool(
        Bool=&SetBoolByFloatComparison(A=&MakeFloatFromVector4(Vector 4=$Emitter.RandomBoolGroups, Channel=::NewEnumerator2), B=0.5)
        True Color=$User.TheColor
        False Color=<rgba 4 3.5 1 0>
       )
       Passthrough Linear Color=&MakeCustomLinearColorFromBool(
        Bool=&SetBoolByFloatComparison(A=&MakeFloatFromVector4(Vector 4=$Emitter.RandomBoolGroups, Channel=::NewEnumerator3), B=0.5)
        True Color=$User.TheColor
        False Color=<rgba 4 3.5 1 0>
       )
       Spawn Group Mask=2
      )
      Spawn Group Mask=1
     )
    )
    Lifetime=&RandomRangeFloat(Minimum=0.1, Maximum=0.2)
    Module.Write Color=true
    Module.Write Lifetime=true
    Module.Write Mass=false
    Module.Write Material Random=true
    Module.Write Position=false
    Module.Write RibbonTwist=false
    Module.Write RibbonWidth=true
    Ribbon Width=7
   )
   &DynamicMaterialParameters(
    Index 0 Param 4=&RandomRangeFloat(Minimum=0, Maximum=1)
    Index 1 Param 1=2
    Index 1 Param 2=1
    Index 1 Param 3=1
    Module.Param0WriteEnabled=false
    Module.Param0WriteEnabled1=true
    Module.Param1WriteEnabled=false
    Module.Param1WriteEnabled1=true
    Module.Param2WriteEnabled=false
    Module.Param2WriteEnabled1=true
    Module.Param3WriteEnabled=true
    Module.Param3WriteEnabled1=false
    Write Parameter Index 1=true
   ) {module=DynamicMaterialParameters001}
  ]
  particleUpdateScript [
   &ParticleState()
   &DynamicMaterialParameters(
    Index 0 Param 1=2
    Index 0 Param 2=&FloatFromCurve(FloatCurve=DataInterface() {propertyValues {Curve {keys [
         {arriveTangent=16.9415, interpMode=RCIM_Cubic, leaveTangent=16.9415, tangentMode=RCTM_User, time=0, value=-1}
         {arriveTangent=3.3333, leaveTangent=3.3333, time=0.25, value=1}
         {arriveTangent=0.7567, interpMode=RCIM_Cubic, leaveTangent=0.7567, tangentMode=RCTM_User, time=0.45, value=1}
         <at 1 -1 RCIM_Cubic>
        ]}}})
    Index 0 Param 3=1
    Index 1 Param 1=&FloatFromCurve(FloatCurve=DataInterface() {propertyValues {Curve {keys [<at 0.4 3>,<at 1 6>]}}}, Scale Curve=1.2)
    Module.Param0WriteEnabled=false
    Module.Param0WriteEnabled1=true
    Module.Param1WriteEnabled=true
    Module.Param1WriteEnabled1=true
    Module.Param2WriteEnabled=true
    Module.Param2WriteEnabled1=true
    Module.Param3WriteEnabled=false
    Module.Param3WriteEnabled1=true
    Write Parameter Index 1=true
   )
   &FN_BoneFlow(
    Angle:NiagaraFloat=$Particles.Initial.FN_Initialize_BoneFlow_Attributes_v2.Angle
    Height:NiagaraFloat=$Particles.Partition.NormalizedID
    Radius:NiagaraFloat=$Particles.Initial.FN_Initialize_BoneFlow_Attributes_v2.Radius
    SkeletalMesh:NiagaraDataInterfaceSkeletalMesh=$User.SkeletalMesh
    Use Sockets Instead=true
    Write Position Attribute=true
   )
   &ScaleColor(
    Scale Mode=::NewEnumerator1
    Scale RGBA=&Vector4FromCurve(Vector4Curve=DataInterface() {propertyValues {
       AlphaCurve {keys [<at 0.706 1>,<at 1 0>]}
       BlueCurve {keys [<at 0 10>,<at 0.1818 1>,<at 0.4428 1>,<at 0.6162 5>]}
       GreenCurve {keys [<at 0 10>,<at 0.1818 1>,<at 0.4428 1>,<at 0.6162 5>]}
       RedCurve {keys [<at 0 10>,<at 0.1818 1>,<at 0.4428 1>,<at 0.6162 5>]}
      }})
    ScaleRGBA=true
   )
   /SetVariables(Particles.DecalColor:LinearColor=&ColorFromCurve(ColorCurve=DataInterface() {propertyValues {
       AlphaCurve {keys [<at 0 0 RCIM_Cubic>,<at 0.2899 1>,<at 1 0 RCIM_Cubic>]}
       BlueCurve {keys [<at 0 0.5677 RCIM_Cubic>]}
       GreenCurve {keys [<at 0 0.928 RCIM_Cubic>]}
       RedCurve {keys [<at 0 1 RCIM_Cubic>]}
      }})) {isSetParameters=true}
   &CameraOffset(Camera Offset Amount=5)
  ]
  rendererClasses [&NiagaraRibbonRendererProperties]
  renderers [{rendererClass=&NiagaraRibbonRendererProperties}]
 }
 BoneFlow_SpawnRate {
  emitterSpawnScript [
   /SetVariables(Emitter.RandomSeed:NiagaraInt32=&RandomInt(Randomness Mode:ENiagaraRandomnessMode=::NewEnumerator0, Random Seed:NiagaraInt32=0, Fixed Random Seed:NiagaraBool=false)) {isSetParameters=true, module=SetVariables_#1}
   &FN_PartitionAttributes(Particles Per Partition=3, Partitions=4)
  ]
  emitterUpdateScript [
   &EmitterState(
    Life Cycle Mode=::NewEnumerator1
    Loop Delay=&RandomRangeFloat(Minimum=0.6, Maximum=1.5)
    Loop Duration=&RandomRangeFloat(Minimum=0.15, Maximum=0.25)
    Recalculate Duration Each Loop=true
    UseLoopDelay=true
   )
   /SetVariables(Emitter.RandomBoolGroups:Vector4f=&RandomRangeVector4(
     Minimum:Vector4f=<0 0 0 0>
     Maximum:Vector4f=<1 1 1 1>
     Evaluation Type:ENiagaraRandomnessEvaluation=::NewEnumerator0
     Recalculate Random Each Loop:NiagaraBool=true
     Randomness Mode:ENiagaraRandomnessMode=::NewEnumerator0
     Random Seed:NiagaraInt32=0
     Fixed Random Seed:NiagaraBool=false
    )) {isSetParameters=true, module=SetVariables_#2}
   &SpawnRate(SpawnRate=&Divide_Float(A=24, B=$Emitter.CurrentLoopDuration), Use Spawn Probability=false)
   &FN_Emitter_SkydiveAttributes() {module=Emitter_SkydiveAttributes}
  ]
  particleSpawnScript [
   &FN_SetSpawnGroup()
   &FN_Particle_Partition_v5(Enable Ribbon ID=true, Pattern:ENiagaraPartitionWaveForm=::NewEnumerator1, Seed:NiagaraInt32=$Emitter.RandomSeed)
   &FN_Initialize_BoneFlow_Attributes_v2(
    Angle:NiagaraFloat=&RandomRangeFloat(Minimum=0, Maximum=1)
    Radius:NiagaraFloat=&FloatFromCurve(
     FloatCurve=DataInterface() {propertyValues {Curve {keys [
         {arriveTangent=5.6652, interpMode=RCIM_Cubic, leaveTangent=5.6652, tangentMode=RCTM_User, time=0, value=0}
         <at 0.5 1 RCIM_Cubic>
         {arriveTangent=-5.8125, interpMode=RCIM_Cubic, leaveTangent=-5.8125, tangentMode=RCTM_User, time=1, value=0}
        ]}}}
     CurveIndex=$Particles.Partition.NormalizedID
     Scale Curve=&RandomRangeFloat(Minimum=8, Maximum=10)
    )
   )
   &InitializeRibbon(
    Color=&MaskLinearColorBySpawnGroup(
     Masked Linear Color=&MakeCustomLinearColorFromBool(
      Bool=&SetBoolByFloatComparison(A=&MakeFloatFromVector4(Vector 4=$Emitter.RandomBoolGroups, Channel=::NewEnumerator0), B=0.3)
      True Color=$User.TheColor
      False Color=<rgba 4 3.5 1 0>
     )
     Passthrough Linear Color=&MaskLinearColorBySpawnGroup(
      Masked Linear Color=&MakeCustomLinearColorFromBool(
       Bool=&SetBoolByFloatComparison(A=&MakeFloatFromVector4(Vector 4=$Emitter.RandomBoolGroups, Channel=::NewEnumerator1), B=0.3)
       True Color=$User.TheColor
       False Color=<rgba 4 3.5 1 0>
      )
      Passthrough Linear Color=&MaskLinearColorBySpawnGroup(
       Masked Linear Color=&MakeCustomLinearColorFromBool(
        Bool=&SetBoolByFloatComparison(A=&MakeFloatFromVector4(Vector 4=$Emitter.RandomBoolGroups, Channel=::NewEnumerator2), B=0.3)
        True Color=$User.TheColor
        False Color=<rgba 4 3.5 1 0>
       )
       Passthrough Linear Color=&MakeCustomLinearColorFromBool(
        Bool=&SetBoolByFloatComparison(A=&MakeFloatFromVector4(Vector 4=$Emitter.RandomBoolGroups, Channel=::NewEnumerator3), B=0.3)
        True Color=$User.TheColor
        False Color=<rgba 4 3.5 1 0>
       )
       Spawn Group Mask=2
      )
      Spawn Group Mask=1
     )
    )
    Lifetime=0.3
    Module.Write Color=true
    Module.Write Lifetime=true
    Module.Write Mass=false
    Module.Write Material Random=true
    Module.Write Position=false
    Module.Write RibbonTwist=false
    Module.Write RibbonWidth=true
    Ribbon Width=10
   )
   &DynamicMaterialParameters(
    Index 0 Param 4=&RandomRangeFloat(Minimum=0, Maximum=1)
    Index 1 Param 1=2
    Index 1 Param 2=1
    Index 1 Param 3=1
    Module.Param0WriteEnabled=false
    Module.Param0WriteEnabled1=true
    Module.Param1WriteEnabled=false
    Module.Param1WriteEnabled1=true
    Module.Param2WriteEnabled=false
    Module.Param2WriteEnabled1=true
    Module.Param3WriteEnabled=true
    Module.Param3WriteEnabled1=false
    Write Parameter Index 1=true
   ) {module=DynamicMaterialParameters001}
   /Niagara/Modules/Spawn/Initialization/InitializeParticle(
    Color=<rgba 1 0.928 0.5677 1>
    Lifetime=$Particles.Lifetime
    Module.Write Color=false
    Module.Write Lifetime=true
    Module.Write Mass=true
    Module.Write Material Random=true
    Module.Write Position=false
    Module.Write Scale=false
    Module.Write SpriteRotation=false
    Module.Write SpriteSize=true
    Position=&SimulationPosition()
    Sprite Size=&Vector2DFromFloat(Value=&Multiply_Float(
      A=35
      B=&MakeCustomFloatFromBool(Bool=&RandomBool(), True Float=1, False Float=0)
     ))
   )
  ]
  particleUpdateScript [
   &ParticleState()
   &DynamicMaterialParameters(
    Index 0 Param 1=1
    Index 0 Param 2=&FloatFromCurve(FloatCurve=DataInterface() {propertyValues {Curve {keys [
         {arriveTangent=16.9415, interpMode=RCIM_Cubic, leaveTangent=16.9415, tangentMode=RCTM_User, time=0, value=-1}
         {arriveTangent=3.3333, leaveTangent=3.3333, time=0.25, value=1}
         {arriveTangent=0.7567, interpMode=RCIM_Cubic, leaveTangent=0.7567, tangentMode=RCTM_User, time=0.45, value=1}
         <at 1 -1 RCIM_Cubic>
        ]}}})
    Index 0 Param 3=1
    Index 1 Param 1=&FloatFromCurve(FloatCurve=DataInterface() {propertyValues {Curve {keys [<at 0.4 3>,<at 1 6>]}}}, Scale Curve=0.9)
    Module.Param0WriteEnabled=false
    Module.Param0WriteEnabled1=true
    Module.Param1WriteEnabled=true
    Module.Param1WriteEnabled1=false
    Module.Param2WriteEnabled=true
    Module.Param2WriteEnabled1=false
    Module.Param3WriteEnabled=false
    Module.Param3WriteEnabled1=false
    Write Parameter Index 1=true
   )
   &FN_BoneFlow(
    Angle:NiagaraFloat=$Particles.Initial.FN_Initialize_BoneFlow_Attributes_v2.Angle
    Height:NiagaraFloat=$Particles.Partition.NormalizedID
    Radius:NiagaraFloat=$Particles.Initial.FN_Initialize_BoneFlow_Attributes_v2.Radius
    SkeletalMesh:NiagaraDataInterfaceSkeletalMesh=$User.SkeletalMesh
    Use Sockets Instead=true
    Write Position Attribute=true
   )
   &ScaleColor(
    Scale Mode=::NewEnumerator1
    Scale RGBA=&Vector4FromCurve(Vector4Curve=DataInterface() {propertyValues {
       AlphaCurve {keys [<at 0.706 1>,<at 1 0>]}
       BlueCurve {keys [<at 0 50>,<at 0.1818 1>,<at 0.4428 1>,<at 0.6162 10>]}
       GreenCurve {keys [<at 0 50>,<at 0.1818 1>,<at 0.4428 1>,<at 0.6162 10>]}
       RedCurve {keys [<at 0 50>,<at 0.1818 1>,<at 0.4428 1>,<at 0.6162 10>]}
      }})
    ScaleRGBA=true
   )
   /SetVariables(Particles.DecalColor:LinearColor=&ColorFromCurve(ColorCurve=DataInterface() {propertyValues {
       AlphaCurve {keys [<at 0 0 RCIM_Cubic>,<at 0.0972 1>,<at 0.6422 0 RCIM_Cubic>]}
       BlueCurve {keys [<at 0 2.1094 RCIM_Cubic>,<at 0.2547 0.4219>]}
       GreenCurve {keys [<at 0 4.5182 RCIM_Cubic>,<at 0.2547 0.9036>]}
       RedCurve {keys [<at 0 5 RCIM_Cubic>,<at 0.2547 1>]}
      }})) {isSetParameters=true}
   &CameraOffset(Camera Offset Amount=15)
  ]
  rendererClasses [&NiagaraRibbonRendererProperties, &NiagaraSpriteRendererProperties]
  renderers [
   {rendererClass=&NiagaraRibbonRendererProperties}
   {rendererClass=&NiagaraSpriteRendererProperties, rendererIndex=1}
  ]
 }
 Decal {
  bEnabled=false
  emitterUpdateScript [
   &EmitterState(
    Life Cycle Mode=::NewEnumerator1
    Loop Duration=&RandomRangeFloat(Minimum=1, Maximum=2.5)
    Recalculate Duration Each Loop=true
    Scalability Mode=::NewEnumerator1
    UseLoopDelay=false
   )
   &SpawnBurst_Instantaneous(Spawn Count=12, Use Spawn Probability=false)
  ]
  particleSpawnScript [
   &FN_SetSpawnGroup_v2(Number of Spawn Groups=2)
   &InitializeParticle(
    Color=<rgba 1 0.928 0.5677 1>
    Lifetime=0.65
    Module.Write Color=true
    Module.Write Lifetime=true
    Module.Write Mass=true
    Module.Write Material Random=true
    Module.Write Position=false
    Module.Write Scale=false
    Module.Write SpriteRotation=false
    Module.Write SpriteSize=true
    Position=&SimulationPosition()
    Sprite Size=&Vector2DFromFloat(Value=20)
   )
   &DynamicMaterialParameters(
    Index 0 Param 1=0.5
    Index 0 Param 2=1
    Index 0 Param 3=1
    Module.Param0WriteEnabled=false
    Module.Param1WriteEnabled=true
    Module.Param2WriteEnabled=true
    Module.Param3WriteEnabled=false
   )
  ]
  particleUpdateScript [
   &ParticleState()
   &DynamicMaterialParameters(
    Index 0 Param 1=&FloatFromCurve(FloatCurve=DataInterface() {propertyValues {Curve {keys [<at 0.4 3>,<at 1 6>]}}})
    Module.Param0WriteEnabled=true
    Module.Param1WriteEnabled=false
    Module.Param2WriteEnabled=false
    Module.Param3WriteEnabled=false
   ) {module=DynamicMaterialParameters001}
   &ScaleColor(
    Scale Alpha=&FloatFromCurve(FloatCurve=DataInterface() {propertyValues {Curve {keys [<at 0 0 RCIM_Cubic>,<at 0.2 1 RCIM_Cubic>,<at 1 0 RCIM_Cubic>]}}}, Scale Curve=0.5)
    ScaleA=true
    ScaleRGB=true
   )
   &CameraOffset(Camera Offset Amount=&RandomRangeFloat(Minimum=15, Maximum=20))
   &SkeletalMeshLocation(
    Inherit Velocity=false
    MaskSpawnGroup=false
    Mesh Sampling Type=::NewEnumerator5
    Module.Override Randomness=false
    Skeletal Mesh=$User.SkeletalMesh
   )
  ]
  rendererClasses [&NiagaraSpriteRendererProperties]
  renderers [{rendererClass=&NiagaraSpriteRendererProperties}]
 }
 Glow {
  emitterUpdateScript [
   &EmitterState()
   &SpawnRate(SpawnRate=6, Use Spawn Probability=false)
  ]
  particleSpawnScript [
   &InitializeParticle(
    Color=<rgba 1 0.928 0.5677 0.05>
    Lifetime=&UniformRangedFloat(Minimum=1, Maximum=1.75)
    Module.Write Color=true
    Module.Write Lifetime=true
    Module.Write Mass=false
    Module.Write Material Random=true
    Module.Write Position=false
    Module.Write Scale=false
    Module.Write SpriteRotation=false
    Module.Write SpriteSize=true
    Position=&SimulationPosition()
    Sprite Rotation=&UniformRangedFloat(Minimum=-180, Maximum=180)
    Sprite Size=&Vector2DFromFloat(Value=&UniformRangedFloat(Minimum=30, Maximum=35))
   )
   &CameraOffset(Camera Offset Amount=&RandomRangeFloat(Minimum=-5, Maximum=10))
  ]
  particleUpdateScript [
   &ParticleState()
   &ScaleColor(
    Scale Alpha=&FloatFromCurve(FloatCurve=DataInterface() {propertyValues {Curve {keys [
         <at 0 0 RCIM_Cubic>
         {arriveTangent=0.1462, interpMode=RCIM_Cubic, leaveTangent=0.1462, time=0.169, value=1.0042}
         {arriveTangent=-1.2044, interpMode=RCIM_Cubic, leaveTangent=-1.2044, time=0.7656, value=0.1119}
         <at 1.0011 0.002>
        ]}}})
    Scale RGB=&VectorFromCurve(VectorCurve=DataInterface() {propertyValues {
       XCurve {keys [<at -0.0011 1.0067>,<at 0.8 0.9431>]}
       YCurve {keys [<at -0.0011 1.0067>,<at 0.7985 0.3834>]}
       ZCurve {keys [<at -0.0011 1.0067>,<at 0.7995 0.1189>]}
      }})
    ScaleA=true
    ScaleRGB=false
   ) {module=ScaleColor001}
   &ScaleSpriteSize(Scale Factor=&Vector2DFromFloat(Value=&FloatFromCurve(FloatCurve=DataInterface() {propertyValues {Curve {keys [
          {arriveTangent=7.5764, interpMode=RCIM_Cubic, leaveTangent=7.5765, tangentMode=RCTM_User, time=0, value=0.02}
          {arriveTangent=1.2063, interpMode=RCIM_Cubic, leaveTangent=1.2063, tangentMode=RCTM_User, time=0.18, value=0.99}
          <at 1 0.5 RCIM_Cubic>
         ]}}})))
   &SkeletalMeshLocation(
    Inherit Velocity=false
    MaskSpawnGroup=false
    Mesh Sampling Type=::NewEnumerator5
    Sampled Position Offset=<-2.5 0 0>
    Sampled Position Offset Coordinate Space:ENiagaraMeshTransforms=::NewEnumerator3
    Skeletal Mesh=$User.SkeletalMesh
    Socket Sampling Mode=::NewEnumerator5
   )
  ]
  rendererClasses [&NiagaraSpriteRendererProperties]
  renderers [{rendererClass=&NiagaraSpriteRendererProperties}]
 }
 Lightning_Sprites {
  emitterUpdateScript [
   &EmitterState()
   &SpawnRate(SpawnRate=1.5, Use Spawn Probability=false)
  ]
  particleSpawnScript [
   &InitializeParticle(
    Color=$User.TheColor
    Lifetime=&UniformRangedFloat(Minimum=0.3, Maximum=0.5)
    Module.Write Color=true
    Module.Write Lifetime=true
    Module.Write Mass=true
    Module.Write Material Random=true
    Module.Write Position=false
    Module.Write Scale=false
    Module.Write SpriteRotation=true
    Module.Write SpriteSize=true
    Position=&SimulationPosition()
    Sprite Rotation=&UniformRangedFloat(Minimum=0, Maximum=360)
    Sprite Size=&Vector2DFromFloat(Value=&Lerp_Float(A=8, B=12, Alpha=1))
   )
   &SphereLocation(
    MaskSpawnGroup=false
    Offset=<0 -2 0>
    Override Seed=false
    Sphere Radius=30
    Surface Only=false
    Surface Only Band Thickness=2
    UseNonUniformScale=false
    UseOffset=true
    UseSphereOrientation=false
   )
   &DynamicMaterialParameters(
    Index 0 Param 1=&UniformRangedFloat(Minimum=0, Maximum=1)
    Index 0 Param 2=0.1
    Index 0 Param 3=1
    Module.Param0WriteEnabled=true
    Module.Param1WriteEnabled=true
    Module.Param2WriteEnabled=true
    Module.Param3WriteEnabled=false
   ) {module=DynamicMaterialParameters001}
   &CameraOffset(Camera Offset Amount=&UniformRangedFloat(Minimum=-4, Maximum=25))
  ]
  particleUpdateScript [
   &ParticleState()
   &SubUVAnimation(End Frame=3, SubUV Animation Mode:ENiagara_SubUVLookupMode=::NewEnumerator2)
   &DynamicMaterialParameters(
    Index 0 Param 3=&FloatFromCurve(FloatCurve=DataInterface() {propertyValues {Curve {keys [<at 0 1 RCIM_Cubic>,<at 0.2 0.025 RCIM_Cubic>,<at 1 1>]}}})
    Module.Param0WriteEnabled=false
    Module.Param1WriteEnabled=false
    Module.Param2WriteEnabled=true
    Module.Param3WriteEnabled=false
   )
   &SkeletalMeshLocation(
    Inherit Velocity=false
    MaskSpawnGroup=false
    Mesh Sampling Type=::NewEnumerator5
    Module.Override Randomness=false
    Sampled Position Offset=&RandomRangeVector(Minimum=<-1 -1 -1>, Maximum=<1 1 1>)
    Skeletal Mesh=$User.SkeletalMesh
   )
  ]
  rendererClasses [&NiagaraSpriteRendererProperties]
  renderers [{rendererClass=&NiagaraSpriteRendererProperties}]
 }
}
summary {
 systemName=NS_Pickaxe_GrilledCheese
 emitters [
  {emitterName=BoneFlow_Burst, rendererClasses [&NiagaraRibbonRendererProperties]}
  {
   emitterName=BoneFlow_SpawnRate
   rendererClasses [&NiagaraRibbonRendererProperties, &NiagaraSpriteRendererProperties]
  }
  {bEnabled=false, emitterName=Decal, rendererClasses [&NiagaraSpriteRendererProperties]}
  {emitterName=Glow, rendererClasses [&NiagaraSpriteRendererProperties]}
  {emitterName=Lightning_Sprites, rendererClasses [&NiagaraSpriteRendererProperties]}
 ]
 userVariables [
  User.PlayerSpeed {type=NiagaraFloat, defaultValue=0, description=""}
  User.SkeletalMesh {
   type=NiagaraDataInterfaceSkeletalMesh
   defaultValue {
    struct=NiagaraExt_VariableValue_DataInterface
    value {dataInterface=&self:NiagaraDataInterfaceSkeletalMesh_0}
   }
  }
  User.TheColor {type=LinearColor, description="", defaultValue {struct=LinearColor, value=<rgba 2 2 0.7188 1>}}
 ]
}
